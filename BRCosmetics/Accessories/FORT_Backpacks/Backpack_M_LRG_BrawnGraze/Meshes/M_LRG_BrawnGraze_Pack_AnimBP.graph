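# Anim Blueprint for the BrawnGraze backpack, adding secondary motion and pose offsets to its dangling body.
# AnimGraph chains spring-based Anim Dynamics on head, spine, tail, legs and arms, then stacks additive bone rotations on arms and legs.
# Each update, EventGraph stores offsets mapped from owner spine bend, parameter-asset gravity, and skydive/parachute/jump pose flags.
# AnimGraph shows constant pin values rather than reads of those variables.
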AnimBlueprint /BRCosmetics/Accessories/FORT_Backpacks/Backpack_M_LRG_BrawnGraze/Meshes/M_LRG_BrawnGraze_Pack_AnimBP
target_skeleton=&/M_LRG_BrawnGraze_Pack_Skeleton
graphs [
 AnimGraph {nodes [
   "Transform (Modify) Bone - Bone: dyn_spine"!(ComponentPose, bAlphaBoolEnabled:bool=True, Translation="0,0,0", Rotation="-12.147,0.000,-0.000", Scale="1,1,1") -> (Pose) {
    alphaInputType=Bool
    rotationMode=BMM_Additive
    rotationSpace=BCS_BoneSpace
    translationMode=BMM_Additive
    translationSpace=BCS_BoneSpace
    boneToModify {boneName=dyn_spine}
    rotation {pitch=-12.147}
   }
   n1 = Anim Dynamics - Bone: dyn_head!(ComponentPose=Local To Component(LocalPose=Input Pose()), Alpha=1, GravityOverride:Vector="0,0,0", ExternalForce="0,0,0") -> (Pose) {
    angularDampingOverride=1
    angularSpringConstant=500
    bAngularSpring=true
    bGravityOverrideInSimSpace=true
    bOverrideAngularDamping=true
    bUseGravityOverride=true
    gravityScale=0.35
    simulationSpace=RootRelative
    boundBone {boneName=dyn_head}
    componentAppliedLinearAccClamp=<3000 3000 3000>
    componentLinearAccScale=<0.5 0.5 0.5>
    componentLinearVelScale=<0.5 0.5 0.5>
    gravityOverride {z=-980}
    physicsBodyDefinitions [{
      boundBone {boneName=dyn_head}
      boxExtents=<30 30 30>
      constraintSetup {angularLimitsMax=<8 7 7>, angularLimitsMin=<-11 -11 -5>, angularTarget=<1 0 0>}
      localJointOffset=<2 0 0>
     }]
   }
   n2 = Anim Dynamics - Bone: dyn_spine!(ComponentPose=@n1, Alpha=1, GravityOverride:Vector="0,0,-980", ExternalForce="0,0,0") -> (Pose) {
    angularDampingOverride=1
    angularSpringConstant=150
    bAngularSpring=true
    bOverrideAngularDamping=true
    bUseGravityOverride=true
    gravityScale=0.35
    simulationSpace=RootRelative
    boundBone {boneName=dyn_spine}
    componentAppliedLinearAccClamp=<3000 3000 3000>
    componentLinearAccScale=<0.4 0.4 0.4>
    componentLinearVelScale=<0.4 0.4 0.4>
    gravityOverride {z=-980}
    physicsBodyDefinitions [{
      boundBone {boneName=dyn_spine}
      boxExtents=<40 40 40>
      constraintSetup {angularLimitsMax=<15 12 13>, angularLimitsMin=<-15 -7 -13>, angularTarget=<1 0 0>}
      localJointOffset=<4 0 0>
     }]
   }
   n3 = Anim Dynamics - Bone: dyn_tail!(ComponentPose=@n2, Alpha=1, GravityOverride:Vector="0,0,0", ExternalForce="0,0,0") -> (Pose) {
    angularDampingOverride=1
    angularSpringConstant=2000
    bAngularSpring=true
    bOverrideAngularDamping=true
    bUseGravityOverride=true
    gravityScale=0.35
    simulationSpace=RootRelative
    boundBone {boneName=dyn_tail}
    componentAppliedLinearAccClamp=<3000 3000 3000>
    componentLinearAccScale=<0.4 0.4 0.4>
    componentLinearVelScale=<0.4 0.4 0.4>
    gravityOverride {z=-980}
    physicsBodyDefinitions [{
      boundBone {boneName=dyn_tail}
      boxExtents=<30 30 30>
      constraintSetup {angularLimitsMax=<85 20 5>, angularLimitsMin=<-85 -40 -5>, angularTarget=<1 0 0>}
      localJointOffset=<1 0 0>
     }]
   }
   n4 = Anim Dynamics - Bone: dyn_leg_1_l!(ComponentPose=@n3, Alpha=1, GravityOverride:Vector="0,0,-980", ExternalForce="0,0,0") -> (Pose) {
    angularDampingOverride=1
    angularSpringConstant=11250
    bAngularSpring=true
    bOverrideAngularDamping=true
    bUseGravityOverride=true
    gravityScale=0.35
    simulationSpace=BoneRelative
    boundBone {boneName=dyn_leg_1_l}
    componentAppliedLinearAccClamp=<3000 3000 3000>
    componentLinearAccScale=<0.5 0.5 0.5>
    componentLinearVelScale=<0.5 0.5 0.5>
    gravityOverride {z=-980}
    physicsBodyDefinitions [{
      boundBone {boneName=dyn_leg_1_l}
      boxExtents=<20 20 20>
      constraintSetup {angularLimitsMax=<45 5 55>, angularLimitsMin=<-9 -14 -1>, angularTarget=<1 0 0>}
      localJointOffset=<3.8 0 0>
     }]
   }
   n5 = Anim Dynamics - Bone: dyn_leg_2_l!(ComponentPose=@n4, Alpha=1, GravityOverride:Vector="0,0,-980", ExternalForce="0,0,0") -> (Pose) {
    angularDampingOverride=1
    angularSpringConstant=14250
    bAngularSpring=true
    bOverrideAngularDamping=true
    bUseGravityOverride=true
    gravityScale=0.35
    simulationSpace=RootRelative
    boundBone {boneName=dyn_leg_2_l}
    componentAppliedLinearAccClamp=<3000 3000 3000>
    componentLinearAccScale=<0.5 0.5 0.5>
    componentLinearVelScale=<0.5 0.5 0.5>
    gravityOverride {z=-980}
    physicsBodyDefinitions [{
      boundBone {boneName=dyn_leg_2_l}
      boxExtents=<20 20 20>
      constraintSetup {angularLimitsMax=<20 3 30>, angularLimitsMin=<0 -11 -5>, angularTarget=<1 0 0>}
      localJointOffset=<1.9 0 0>
     }]
   }
   n6 = Anim Dynamics - Bone: dyn_leg_1_r!(ComponentPose=@n5, Alpha=1, GravityOverride:Vector="0,0,-980", ExternalForce="0,0,0") -> (Pose) {
    angularDampingOverride=0.9999
    angularSpringConstant=11250
    bAngularSpring=true
    bOverrideAngularDamping=true
    bUseGravityOverride=true
    gravityScale=0.35
    simulationSpace=BoneRelative
    boundBone {boneName=dyn_leg_1_r}
    componentAppliedLinearAccClamp=<3000 3000 3000>
    componentLinearAccScale=<0.5 0.5 0.5>
    componentLinearVelScale=<0.5 0.5 0.5>
    gravityOverride {z=-980}
    physicsBodyDefinitions [{
      boundBone {boneName=dyn_leg_1_r}
      boxExtents=<20 20 20>
      constraintSetup {angularLimitsMax=<19 0 5>, angularLimitsMin=<-11 -22 -40>, angularTarget=<1 0 0>}
      localJointOffset=<4 0 0>
     }]
   }
   n7 = Anim Dynamics - Bone: dyn_leg_2_r!(ComponentPose=@n6, Alpha=1, GravityOverride:Vector="0,0,-980", ExternalForce="0,0,0") -> (Pose) {
    angularDampingOverride=1
    angularSpringConstant=12000
    bAngularSpring=true
    bOverrideAngularDamping=true
    bUseGravityOverride=true
    gravityScale=0.35
    simulationSpace=RootRelative
    boundBone {boneName=dyn_leg_2_r}
    componentAppliedLinearAccClamp=<3000 3000 3000>
    componentLinearAccScale=<0.5 0.5 0.5>
    componentLinearVelScale=<0.5 0.5 0.5>
    gravityOverride {z=-980}
    physicsBodyDefinitions [{
      boundBone {boneName=dyn_leg_2_r}
      boxExtents=<20 20 20>
      constraintSetup {angularLimitsMax=<20 0 38>, angularLimitsMin=<0 -25 -15>, angularTarget=<1 0 0>}
      localJointOffset=<2 0 0>
     }]
   }
   n8 = Anim Dynamics - Bone: dyn_arm_1_l!(ComponentPose=@n7, Alpha=1, GravityOverride:Vector="0,0,-980", ExternalForce="0,0,0") -> (Pose) {
    angularDampingOverride=1
    angularSpringConstant=879
    bAngularSpring=true
    bOverrideAngularDamping=true
    bUseGravityOverride=true
    gravityScale=0.35
    boundBone {boneName=dyn_arm_1_l}
    componentAppliedLinearAccClamp=<3000 3000 3000>
    componentLinearAccScale=<0.3 0.3 0.3>
    componentLinearVelScale=<0.3 0.3 0.3>
    gravityOverride {z=-980}
    physicsBodyDefinitions [{
      boundBone {boneName=dyn_arm_1_l}
      boxExtents=<20 20 20>
      constraintSetup {angularLimitsMax=<5 30 53>, angularLimitsMin=<-8 -44 4>, angularTarget=<1 0 0>}
      localJointOffset=<4 0 0>
     }]
    simSpaceSettings {maxAngularAcceleration=5, maxAngularVelocity=3, simSpaceAngularAlpha=1}
   }
   n9 = Anim Dynamics - Bone: dyn_arm_1_r!(ComponentPose=@n8, Alpha=1, GravityOverride:Vector="0,0,-980", ExternalForce="0,0,0") -> (Pose) {
    angularDampingOverride=0.9999
    angularSpringConstant=825
    bAngularSpring=true
    bOverrideAngularDamping=true
    bUseGravityOverride=true
    gravityScale=0.35
    boundBone {boneName=dyn_arm_1_r}
    componentAppliedLinearAccClamp=<3000 3000 3000>
    componentLinearAccScale=<0.3 0.3 0.3>
    componentLinearVelScale=<0.3 0.3 0.3>
    gravityOverride {z=-980}
    physicsBodyDefinitions [{
      boundBone {boneName=dyn_arm_1_r}
      boxExtents=<20 20 20>
      constraintSetup {angularLimitsMax=<8 44 -6>, angularLimitsMin=<-3 -30 -55>, angularTarget=<1 0 0>}
      localJointOffset=<4 0 0>
     }]
    simSpaceSettings {maxAngularAcceleration=5, maxAngularVelocity=3, simSpaceAngularAlpha=1}
    sphericalLimits [{limitRadius=0, drivingBone {boneName=dyn_arm_1_r}, sphereLocalOffset=<0 0 0>}]
   }
   n10 = "Transform (Modify) Bone - Bone: dyn_arm_1_l"!(ComponentPose=@n9, bAlphaBoolEnabled:bool=True, Translation="0,0,0", Rotation="-15,23,8", Scale="1,1,1") -> (Pose) {
    alphaInputType=Bool
    rotationMode=BMM_Additive
    rotationSpace=BCS_BoneSpace
    translationMode=BMM_Additive
    translationSpace=BCS_BoneSpace
    alphaBoolBlend {blendInTime=3, blendOutTime=3}
    boneToModify {boneName=dyn_arm_1_l}
    rotation=<rot -15 23 8>
   }
   n11 = "Transform (Modify) Bone - Bone: dyn_arm_1_r"!(ComponentPose=@n10, bAlphaBoolEnabled:bool=True, Translation="0,0,0", Rotation="-16,-20,-5", Scale="1,1,1") -> (Pose) {
    alphaInputType=Bool
    rotationMode=BMM_Additive
    rotationSpace=BCS_BoneSpace
    translationMode=BMM_Additive
    translationSpace=BCS_BoneSpace
    alphaBoolBlend {blendInTime=3, blendOutTime=3}
    boneToModify {boneName=dyn_arm_1_r}
    rotation=<rot -16 -20 -5>
   }
   n12 = "Transform (Modify) Bone - Bone: dyn_leg_1_l"!(ComponentPose=@n11, bAlphaBoolEnabled:bool=True, Translation="0,0,0", Rotation="-15,25,0", Scale="1,1,1") -> (Pose) {
    alphaInputType=Bool
    rotationMode=BMM_Additive
    rotationSpace=BCS_BoneSpace
    translationMode=BMM_Additive
    translationSpace=BCS_BoneSpace
    alphaBoolBlend {blendInTime=3, blendOutTime=3}
    boneToModify {boneName=dyn_leg_1_l}
    rotation {pitch=-15, yaw=25}
   }
   n13 = "Transform (Modify) Bone - Bone: dyn_leg_2_l"!(ComponentPose=@n12, bAlphaBoolEnabled:bool=True, Translation="0,0,0", Rotation="15,-10,0", Scale="1,1,1") -> (Pose) {
    alphaInputType=Bool
    rotationMode=BMM_Additive
    rotationSpace=BCS_BoneSpace
    translationMode=BMM_Additive
    translationSpace=BCS_BoneSpace
    alphaBoolBlend {blendInTime=3, blendOutTime=3}
    boneToModify {boneName=dyn_leg_2_l}
    rotation {pitch=15, yaw=-10}
   }
   n14 = "Transform (Modify) Bone - Bone: dyn_leg_1_r"!(ComponentPose=@n13, bAlphaBoolEnabled:bool=True, Translation="0,0,0", Rotation="-15,-25,0", Scale="1,1,1") -> (Pose) {
    alphaInputType=Bool
    rotationMode=BMM_Additive
    rotationSpace=BCS_BoneSpace
    translationMode=BMM_Additive
    translationSpace=BCS_BoneSpace
    alphaBoolBlend {blendInTime=3, blendOutTime=3}
    boneToModify {boneName=dyn_leg_1_r}
    rotation {pitch=-15, yaw=-25}
   }
   n15 = "Transform (Modify) Bone - Bone: dyn_leg_2_r"!(ComponentPose=@n14, bAlphaBoolEnabled:bool=True, Translation="0,0,0", Rotation="-20,0,0", Scale="1,1,1") -> (Pose) {
    alphaInputType=Bool
    rotationMode=BMM_Additive
    rotationSpace=BCS_BoneSpace
    translationMode=BMM_Additive
    translationSpace=BCS_BoneSpace
    alphaBoolBlend {blendInTime=3, blendOutTime=3}
    boneToModify {boneName=dyn_leg_2_r}
    rotation {pitch=-20}
   }
   n16 = "Transform (Modify) Bone - Bone: dyn_leg_1_l"(ComponentPose=@n15, bAlphaBoolEnabled=True, Translation="0,0,0", Rotation="50,0,0", Scale="1,1,1") {
    alphaInputType=Bool
    rotationMode=BMM_Additive
    rotationSpace=BCS_BoneSpace
    translationMode=BMM_Additive
    translationSpace=BCS_BoneSpace
    boneToModify {boneName=dyn_leg_1_l}
    rotation {pitch=50}
   }
   n17 = "Transform (Modify) Bone - Bone: dyn_leg_1_r"(ComponentPose=@n16, bAlphaBoolEnabled=True, Translation="0,0,0", Rotation="50,0,0", Scale="1,1,1") {
    alphaInputType=Bool
    rotationMode=BMM_Additive
    rotationSpace=BCS_BoneSpace
    translationMode=BMM_Additive
    translationSpace=BCS_BoneSpace
    boneToModify {boneName=dyn_leg_1_r}
    rotation {pitch=50}
   }
   n18 = "Transform (Modify) Bone - Bone: dyn_arm_1_l"(ComponentPose=@n17, bAlphaBoolEnabled=True, Translation="0,0,0", Rotation="-5,20,0", Scale="1,1,1") {
    alphaInputType=Bool
    rotationMode=BMM_Additive
    rotationSpace=BCS_BoneSpace
    translationMode=BMM_Additive
    translationSpace=BCS_BoneSpace
    boneToModify {boneName=dyn_arm_1_l}
    rotation {pitch=-5, yaw=20}
   }
   n19 = "Transform (Modify) Bone - Bone: dyn_arm_1_r"(ComponentPose=@n18, bAlphaBoolEnabled=True, Translation="0,0,0", Rotation="-4,-20,0.076", Scale="1,1,1") {
    alphaInputType=Bool
    rotationMode=BMM_Additive
    rotationSpace=BCS_BoneSpace
    translationMode=BMM_Additive
    translationSpace=BCS_BoneSpace
    boneToModify {boneName=dyn_arm_1_r}
    rotation=<rot -4 -20 0.076>
   }
   n20 = "Transform (Modify) Bone - Bone: dyn_arm_1_l"(ComponentPose=@n19, bAlphaBoolEnabled=True, Translation="0,0,0", Rotation="0,6,0", Scale="1,1,1") {
    alphaInputType=Bool
    rotationMode=BMM_Additive
    rotationSpace=BCS_BoneSpace
    translationMode=BMM_Additive
    translationSpace=BCS_BoneSpace
    boneToModify {boneName=dyn_arm_1_l}
    rotation {yaw=6}
   }
   n21 = "Transform (Modify) Bone - Bone: dyn_arm_1_r"(ComponentPose=@n20, bAlphaBoolEnabled=True, Translation="0,0,0", Rotation="0,-7,0", Scale="1,1,1") {
    alphaInputType=Bool
    rotationMode=BMM_Additive
    rotationSpace=BCS_BoneSpace
    translationMode=BMM_Additive
    translationSpace=BCS_BoneSpace
    boneToModify {boneName=dyn_arm_1_r}
    rotation {yaw=-7}
   }
   Output Pose(Result=Component To Local(ComponentPose=@n21)) {layerGroup=None}
  ]}
 EventGraph {nodes [
   Event BlueprintUpdateAnimation()
   > n1 = Sequence!(execute) -> (then_0, then_1, then_2:Exec, then_3:Exec, then_4:Exec)
   n2 = GetOwnerSkeletalMeshComponent()
   n3 = BreakRotator(InRot=BreakTransform(InTransform=GetDeltaTransformFromRefPose(self=@n2, BoneName=spine_05, BaseName=spine_01)).Rotation)
   > Set Backpack Rotate(Backpack Rotate:Rotator=MakeRotator(Pitch=MapRangeClamped(Value=@n3.Yaw, InRangeA=-10, InRangeB=-65, OutRangeB=-50))) -> (Output_Get:Rotator)
   > Set Backpack Rotate Leg(Backpack Rotate Leg:Rotator=MakeRotator(Pitch=MapRangeClamped(Value=@n3.Yaw, InRangeA=-10, InRangeB=-55, OutRangeB=-40))) -> (Output_Get:Rotator)
   > Set Backpack Translation(Backpack Translation:Vector=MakeVector(Y=MapRangeClamped(Value=@n3.Yaw, InRangeA=0, InRangeB=-45, OutRangeB=15))) -> (Output_Get:Vector)
   n4 = "Gravity Override State Function (Parameter Asset) (Nativized)"(
    execute=@n1.then_1
    CharacterPartAnimInstance=Self-Reference()
    GameTimeInSeconds=GetGameTimeInSeconds()
    DebugJointName=spine_05
    Parameters=&/M_MED_BrawnGraze_Pack_Parameters
   )
   > Set Gravity Override(Gravity Override=@n4)
   Set SkyDive Floating Forward(
    execute=@n1.then_2
    SkyDive Floating Forward:bool=OR Boolean!(A=Get bIsParachuteOpened(), B=Get bIsSkyDiving(), C:bool=Get bIsSkydiveFloating()) -> (ReturnValue)
   ) -> (Output_Get:bool)
   n5 = Get bIsJumping()
   n6 = Get bIsParachuteOpened()
   n7 = Get bIsSkyDiving()
   n8 = Get bIsSkydiveFloating()
   Set SkyDive Floating Arms(
    execute=@n1.then_3
    SkyDive Floating Arms:bool=AND Boolean!(A=NOT Boolean(A=@n6), B=NOT Boolean(A=@n5), C:bool=OR Boolean!(A=@n7, B=@n8) -> (ReturnValue)) -> (ReturnValue)
   ) -> (Output_Get:bool)
   TryGetPawnOwner()
  ]}
]
variables [
 Backpack Rotate {type=Rotator}
 Backpack Rotate Leg {type=Rotator}
 Backpack Translation {type=Vector}
 Gravity Override {type=Vector}
 SkyDive Floating Arms {type=boolean}
 SkyDive Floating Forward {type=boolean, default=true}
 SkyDive Forward {type=boolean}
]
